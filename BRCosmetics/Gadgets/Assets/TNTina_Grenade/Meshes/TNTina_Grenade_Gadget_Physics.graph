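# Physics asset for the TNTina grenade gadget mesh, a cosmetic gadget, providing simple collision shapes for its skeleton.
# A tilted, offset capsule covers the root bone, while similarly sized spheres cover five numbered ball bones and the free pin center bone.
# No constraints link the bodies, so the asset defines shapes only, with no joint limits between them.
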PhysicsAsset /BRCosmetics/Gadgets/Assets/TNTina_Grenade/Meshes/TNTina_Grenade_Gadget_Physics
bodies {
 ball_01 [{shapeType=Sphere, radius=7.25}]
 ball_02 [{shapeType=Sphere, radius=8}]
 ball_03 [{shapeType=Sphere, radius=8}]
 ball_04 [{shapeType=Sphere, radius=8.25}]
 ball_05 [{shapeType=Sphere, radius=8.25}]
 free_pin_center [{shapeType=Sphere, radius=8}]
 root [{shapeType=Capsule, length=8.6624, radius=6.3833, center=<0.6818 -1.2289 2.8541>, rotation=<rot -21.5143 -90 -13.652>}]
}
constraints []
